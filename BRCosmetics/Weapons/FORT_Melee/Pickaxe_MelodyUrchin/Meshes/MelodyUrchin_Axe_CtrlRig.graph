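# Control Rig for the MelodyUrchin pickaxe that adds procedural secondary motion to its tongue and teeth bones.
# From BeginExecution, one linear sequence runs ChainHarmonicsPerItem (pendulum, reach, wave) on the tongue and six handle-teeth chains.
# ItemHarmonics then oscillates nine of the seventeen individual teeth, mostly staggered by wave offsets.
# A second tongue harmonics node is left unconnected, and every chain node has debug drawing enabled.
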ControlRigBlueprint /BRCosmetics/Weapons/FORT_Melee/Pickaxe_MelodyUrchin/Meshes/MelodyUrchin_Axe_CtrlRig
graphs [
 RigVMFunctionLibrary {links [], nodes []}
 RigVMModel {nodes [
   RigUnit_BeginExecution = &RigUnit_BeginExecution()
   /Script/ControlRig.RigUnit_ChainHarmonicsPerItem(
    ChainRoot=(Type=Bone,Name="dyn_tongue_01")
    DrawWorldOffset=(Rotation=(X=0,Y=0,Z=0,W=1),Translation=(X=0,Y=0,Z=0),Scale3D=(X=1,Y=1,Z=1))
    Pendulum=(bEnabled=True,PendulumStiffness=2,PendulumGravity=(X=0,Y=0,Z=0),PendulumBlend=0.75,PendulumDrag=0.98,PendulumMaximum=0.1,UnwindAxis=(X=0,Y=1,Z=0),UnwindMinimum=0.2,UnwindMaximum=0.05)
    Reach=(bEnabled=True,ReachTarget=(X=0,Y=0,Z=0),ReachAxis=(X=1,Y=0,Z=0),ReachMinimum=0,ReachMaximum=0)
    Speed=(X=1,Y=2,Z=2)
    Wave=(bEnabled=True,WaveFrequency=(X=0,Y=-1,Z=-4.5),WaveAmplitude=(X=0.1,Y=1,Z=-1),WaveOffset=(X=0,Y=0,Z=-5),WaveNoise=(X=0,Y=0,Z=0),WaveMinimum=-0.5,WaveMaximum=-1)
    WaveCurve=(EditorCurveData=(Keys=((Time=0.1619,Value=0.3394),(Time=0.3814,Value=0.3559),(Time=0.5431,Value=0.407),(Time=0.703,Value=0.5114),(Time=0.7797,Value=0.5756))))
    bDrawDebug=True
    ExecutePin=@RigUnit_BeginExecution.ExecutePin
   ) {is=struct RigVMUnitNode}
   > &RigUnit_ChainHarmonicsPerItem(
    ChainRoot=(Type=Bone,Name="dyn_teeth_handel_01_00")
    DrawWorldOffset=(Rotation=(X=0,Y=0,Z=0,W=1),Translation=(X=0,Y=0,Z=0),Scale3D=(X=1,Y=1,Z=1))
    Pendulum=(bEnabled=true,PendulumStiffness=1.5,PendulumGravity=(X=0,Y=0,Z=0),PendulumBlend=0.95,PendulumDrag=0.98,PendulumMinimum=0,PendulumMaximum=0.7,PendulumEase=Linear,UnwindAxis=(X=0,Y=1,Z=0),UnwindMinimum=0.2,UnwindMaximum=0.05)
    Reach=(bEnabled=true,ReachTarget=(X=1.5,Y=1.5,Z=1.5),ReachAxis=(X=1,Y=0,Z=0),ReachMinimum=0,ReachMaximum=0,ReachEase=Linear)
    Speed=(X=2,Y=2,Z=2)
    Wave=(bEnabled=True,WaveFrequency=(X=0,Y=0,Z=0),WaveAmplitude=(X=0,Y=5,Z=0),WaveOffset=(X=0,Y=0,Z=0),WaveNoise=(X=0,Y=0,Z=0),WaveMinimum=0.9,WaveMaximum=0,WaveEase=SineEaseIn)
    WaveCurve="(EditorCurveData=(Keys=((InterpMode=RCIM_Cubic,Time=0.0121,Value=0.0862),(InterpMode=RCIM_Cubic,Time=0.2888,Value=0.1693,ArriveTangent=0.0706,LeaveTangent=0.0706),(InterpMode=RCIM_Cubic,Time=0.6097,Value=0.1284,ArriveTangent=0.0033,LeaveTangent=0.0033),(In ...(truncated)"
    bDrawDebug=True
   )
   > &RigUnit_ChainHarmonicsPerItem(
    ChainRoot=(Type=Bone,Name="dyn_teeth_handel_02_00")
    DrawWorldOffset=(Rotation=(X=0,Y=0,Z=0,W=1),Translation=(X=0,Y=0,Z=0),Scale3D=(X=1,Y=1,Z=1))
    Pendulum=(bEnabled=true,PendulumStiffness=1.5,PendulumGravity=(X=0,Y=0,Z=0),PendulumBlend=0.95,PendulumDrag=0.98,PendulumMinimum=0,PendulumMaximum=0.7,PendulumEase=Linear,UnwindAxis=(X=0,Y=1,Z=0),UnwindMinimum=0.2,UnwindMaximum=0.05)
    Reach=(bEnabled=true,ReachTarget=(X=1.5,Y=1.5,Z=1.5),ReachAxis=(X=1,Y=0,Z=0),ReachMinimum=0,ReachMaximum=0,ReachEase=Linear)
    Speed=(X=2,Y=2,Z=2)
    Wave=(bEnabled=True,WaveFrequency=(X=0,Y=1,Z=0),WaveAmplitude=(X=0,Y=7,Z=0),WaveOffset=(X=2,Y=2,Z=2),WaveNoise=(X=0,Y=0,Z=0),WaveMinimum=0.7,WaveMaximum=0,WaveEase=SineEaseIn)
    WaveCurve="(EditorCurveData=(Keys=((InterpMode=RCIM_Cubic,Time=0.0121,Value=0.0862),(InterpMode=RCIM_Cubic,Time=0.2888,Value=0.1693,ArriveTangent=0.0702,LeaveTangent=0.0702),(InterpMode=RCIM_Cubic,Time=0.5817,Value=0.1262,ArriveTangent=-0.0001,LeaveTangent=-0.0001),( ...(truncated)"
    bDrawDebug=True
   )
   > &RigUnit_ChainHarmonicsPerItem(
    ChainRoot=(Type=Bone,Name="dyn_teeth_handel_03_00")
    DrawWorldOffset=(Rotation=(X=0,Y=0,Z=0,W=1),Translation=(X=0,Y=0,Z=0),Scale3D=(X=1,Y=1,Z=1))
    Pendulum=(bEnabled=true,PendulumStiffness=1.5,PendulumGravity=(X=0,Y=0,Z=0),PendulumBlend=0.95,PendulumDrag=0.98,PendulumMinimum=0,PendulumMaximum=0.7,PendulumEase=Linear,UnwindAxis=(X=0,Y=1,Z=0),UnwindMinimum=0.2,UnwindMaximum=0.05)
    Reach=(bEnabled=true,ReachTarget=(X=1.5,Y=1.5,Z=1.5),ReachAxis=(X=1,Y=0,Z=0),ReachMinimum=0,ReachMaximum=0,ReachEase=Linear)
    Speed=(X=2,Y=2,Z=2)
    Wave=(bEnabled=True,WaveFrequency=(X=0,Y=0,Z=0),WaveAmplitude=(X=0,Y=4,Z=0),WaveOffset=(X=1,Y=1,Z=1),WaveNoise=(X=0,Y=0,Z=0),WaveMinimum=0.9,WaveMaximum=0,WaveEase=SineEaseIn)
    WaveCurve="(EditorCurveData=(Keys=((InterpMode=RCIM_Cubic,Time=0.0121,Value=0.0862),(InterpMode=RCIM_Cubic,Time=0.2888,Value=0.1693,ArriveTangent=0.0702,LeaveTangent=0.0702),(InterpMode=RCIM_Cubic,Time=0.5817,Value=0.1262,ArriveTangent=-0.0228,LeaveTangent=-0.0228),( ...(truncated)"
    bDrawDebug=True
   )
   > &RigUnit_ChainHarmonicsPerItem(
    ChainRoot=(Type=Bone,Name="dyn_teeth_handel_04_00")
    DrawWorldOffset=(Rotation=(X=0,Y=0,Z=0,W=1),Translation=(X=0,Y=0,Z=0),Scale3D=(X=1,Y=1,Z=1))
    Pendulum=(bEnabled=true,PendulumStiffness=1.5,PendulumGravity=(X=0,Y=0,Z=0),PendulumBlend=0.95,PendulumDrag=0.98,PendulumMinimum=0,PendulumMaximum=0.7,PendulumEase=Linear,UnwindAxis=(X=0,Y=1,Z=0),UnwindMinimum=0.2,UnwindMaximum=0.05)
    Reach=(bEnabled=true,ReachTarget=(X=1.5,Y=1.5,Z=1.5),ReachAxis=(X=1,Y=0,Z=0),ReachMinimum=0,ReachMaximum=0,ReachEase=Linear)
    Speed=(X=2,Y=2,Z=2)
    Wave=(bEnabled=True,WaveFrequency=(X=0,Y=0,Z=0),WaveAmplitude=(X=0,Y=4,Z=0),WaveOffset=(X=3,Y=3,Z=3),WaveNoise=(X=0,Y=0,Z=0),WaveMinimum=0.9,WaveMaximum=0,WaveEase=SineEaseIn)
    WaveCurve="(EditorCurveData=(Keys=((InterpMode=RCIM_Cubic,Time=0.0121,Value=0.0862),(InterpMode=RCIM_Cubic,Time=0.2888,Value=0.1693,ArriveTangent=0.0702,LeaveTangent=0.0702),(InterpMode=RCIM_Cubic,Time=0.5817,Value=0.1262,ArriveTangent=-0.0228,LeaveTangent=-0.0228),( ...(truncated)"
    bDrawDebug=True
   )
   > &RigUnit_ChainHarmonicsPerItem(
    ChainRoot=(Type=Bone,Name="dyn_teeth_handel_05_00")
    DrawWorldOffset=(Rotation=(X=0,Y=0,Z=0,W=1),Translation=(X=0,Y=0,Z=0),Scale3D=(X=1,Y=1,Z=1))
    Pendulum=(bEnabled=true,PendulumStiffness=1.5,PendulumGravity=(X=0,Y=0,Z=0),PendulumBlend=0.95,PendulumDrag=0.98,PendulumMinimum=0,PendulumMaximum=0.7,PendulumEase=Linear,UnwindAxis=(X=0,Y=1,Z=0),UnwindMinimum=0.2,UnwindMaximum=0.05)
    Reach=(bEnabled=true,ReachTarget=(X=1.5,Y=1.5,Z=1.5),ReachAxis=(X=1,Y=0,Z=0),ReachMinimum=0,ReachMaximum=0,ReachEase=Linear)
    Speed=(X=2,Y=2,Z=2)
    Wave=(bEnabled=True,WaveFrequency=(X=0,Y=0,Z=0),WaveAmplitude=(X=0,Y=4,Z=0),WaveOffset=(X=4,Y=4,Z=4),WaveNoise=(X=0,Y=0,Z=0),WaveMinimum=0.9,WaveMaximum=0,WaveEase=SineEaseIn)
    WaveCurve="(EditorCurveData=(Keys=((InterpMode=RCIM_Cubic,Time=0.0121,Value=0.0862),(InterpMode=RCIM_Cubic,Time=0.2888,Value=0.1693,ArriveTangent=0.0702,LeaveTangent=0.0702),(InterpMode=RCIM_Cubic,Time=0.5817,Value=0.1262,ArriveTangent=-0.0228,LeaveTangent=-0.0228),( ...(truncated)"
    bDrawDebug=True
   )
   > &RigUnit_ChainHarmonicsPerItem(
    ChainRoot=(Type=Bone,Name="dyn_teeth_handel_06_00")
    DrawWorldOffset=(Rotation=(X=0,Y=0,Z=0,W=1),Translation=(X=0,Y=0,Z=0),Scale3D=(X=1,Y=1,Z=1))
    Pendulum=(bEnabled=true,PendulumStiffness=1.5,PendulumGravity=(X=0,Y=0,Z=0),PendulumBlend=0.95,PendulumDrag=0.98,PendulumMinimum=0,PendulumMaximum=0.7,PendulumEase=Linear,UnwindAxis=(X=0,Y=1,Z=0),UnwindMinimum=0.2,UnwindMaximum=0.05)
    Reach=(bEnabled=true,ReachTarget=(X=1.5,Y=1.5,Z=1.5),ReachAxis=(X=1,Y=0,Z=0),ReachMinimum=0,ReachMaximum=0,ReachEase=Linear)
    Speed=(X=2,Y=2,Z=2)
    Wave=(bEnabled=True,WaveFrequency=(X=0,Y=0,Z=0),WaveAmplitude=(X=0,Y=4,Z=0),WaveOffset=(X=1,Y=1,Z=1),WaveNoise=(X=0,Y=0,Z=0),WaveMinimum=0.9,WaveMaximum=0,WaveEase=SineEaseIn)
    WaveCurve="(EditorCurveData=(Keys=((InterpMode=RCIM_Cubic,Time=0.0121,Value=0.0862),(InterpMode=RCIM_Cubic,Time=0.2888,Value=0.1693,ArriveTangent=0.0702,LeaveTangent=0.0702),(InterpMode=RCIM_Cubic,Time=0.5817,Value=0.1262,ArriveTangent=-0.0228,LeaveTangent=-0.0228),( ...(truncated)"
    bDrawDebug=True
   )
   > /Script/ControlRig.RigUnit_ItemHarmonics(
    RotationOrder=YZX
    Targets=((Item=(Type=Bone,Name="dyn_teeth_01"),Ratio=0))
    WaveAmplitude=(X=0,Y=5,Z=0)
    WaveEase=SineEaseIn
    WaveFrequency=(X=0,Y=0,Z=0)
    WaveMaximum=1
    WaveMinimum=0.5
    WaveNoise=(X=0,Y=0,Z=0)
    WaveOffset=(X=0,Y=3,Z=5)
    WaveSpeed=(X=7,Y=7,Z=7)
   ) {is=struct RigVMUnitNode}
   > &RigUnit_ItemHarmonics(
    RotationOrder=YZX
    Targets=((Item=(Type=Bone,Name="dyn_teeth_05"),Ratio=0))
    WaveAmplitude=(X=0,Y=5,Z=0)
    WaveEase=SineEaseIn
    WaveFrequency=(X=0,Y=0,Z=0)
    WaveMaximum=1
    WaveMinimum=0.5
    WaveNoise=(X=0,Y=0,Z=0)
    WaveOffset=(X=0,Y=10,Z=0)
    WaveSpeed=(X=7,Y=7,Z=7)
   )
   > &RigUnit_ItemHarmonics(
    RotationOrder=YZX
    Targets=((Item=(Type=Bone,Name="dyn_teeth_02"),Ratio=0))
    WaveAmplitude=(X=0,Y=5,Z=0)
    WaveEase=SineEaseIn
    WaveFrequency=(X=0,Y=0,Z=0)
    WaveMaximum=1
    WaveMinimum=0.5
    WaveNoise=(X=0,Y=0,Z=0)
    WaveOffset=(X=0,Y=0,Z=0)
    WaveSpeed=(X=7,Y=7,Z=7)
   )
   > &RigUnit_ItemHarmonics(
    RotationOrder=YZX
    Targets=((Item=(Type=Bone,Name="dyn_teeth_09"),Ratio=0))
    WaveAmplitude=(X=0,Y=0,Z=5)
    WaveEase=SineEaseIn
    WaveFrequency=(X=0,Y=0,Z=0)
    WaveMaximum=1
    WaveMinimum=0.5
    WaveNoise=(X=0,Y=0,Z=0)
    WaveOffset=(X=0,Y=0,Z=15)
    WaveSpeed=(X=7,Y=7,Z=7)
   )
   > &RigUnit_ItemHarmonics(
    RotationOrder=YZX
    Targets=((Item=(Type=Bone,Name="dyn_teeth_08"),Ratio=0))
    WaveAmplitude=(X=0,Y=0,Z=5)
    WaveEase=SineEaseIn
    WaveFrequency=(X=0,Y=0,Z=0)
    WaveMaximum=1
    WaveMinimum=0.5
    WaveNoise=(X=0,Y=0,Z=0)
    WaveOffset=(X=0,Y=0,Z=5)
    WaveSpeed=(X=7,Y=7,Z=7)
   )
   > &RigUnit_ItemHarmonics(
    RotationOrder=YZX
    Targets=((Item=(Type=Bone,Name="dyn_teeth_11"),Ratio=0))
    WaveAmplitude=(X=0,Y=0,Z=6)
    WaveEase=SineEaseIn
    WaveFrequency=(X=0,Y=0,Z=0)
    WaveMaximum=1
    WaveMinimum=0.5
    WaveNoise=(X=0,Y=0,Z=0)
    WaveOffset=(X=0,Y=0,Z=0)
    WaveSpeed=(X=7,Y=7,Z=7)
   )
   > &RigUnit_ItemHarmonics(
    RotationOrder=YZX
    Targets=((Item=(Type=Bone,Name="dyn_teeth_15"),Ratio=0))
    WaveAmplitude=(X=0,Y=5,Z=0)
    WaveEase=SineEaseIn
    WaveFrequency=(X=0,Y=0,Z=0)
    WaveMaximum=1
    WaveMinimum=0.5
    WaveNoise=(X=0,Y=0,Z=0)
    WaveOffset=(X=3,Y=0,Z=0)
    WaveSpeed=(X=7,Y=7,Z=7)
   )
   > &RigUnit_ItemHarmonics(
    RotationOrder=YZX
    Targets=((Item=(Type=Bone,Name="dyn_teeth_07"),Ratio=0))
    WaveAmplitude=(X=0,Y=0,Z=5)
    WaveEase=SineEaseIn
    WaveFrequency=(X=0,Y=0,Z=0)
    WaveMaximum=1
    WaveMinimum=0.5
    WaveNoise=(X=0,Y=0,Z=0)
    WaveOffset=(X=0,Y=0,Z=10)
    WaveSpeed=(X=7,Y=7,Z=7)
   )
   > &RigUnit_ItemHarmonics(
    RotationOrder=YZX
    Targets=((Item=(Type=Bone,Name="dyn_teeth_03"),Ratio=0))
    WaveAmplitude=(X=0,Y=0,Z=5)
    WaveEase=SineEaseIn
    WaveFrequency=(X=0,Y=0,Z=0)
    WaveMaximum=1
    WaveMinimum=0.5
    WaveNoise=(X=0,Y=0,Z=0)
    WaveOffset=(X=0,Y=0,Z=10)
    WaveSpeed=(X=7,Y=7,Z=7)
   )
   &RigUnit_ChainHarmonicsPerItem(
    ChainRoot=(Type=Bone,Name="dyn_tongue_01")
    DrawWorldOffset=(Rotation=(X=0,Y=0,Z=0,W=1),Translation=(X=0,Y=0,Z=0),Scale3D=(X=1,Y=1,Z=1))
    ExecutePin=()
    Pendulum=(bEnabled=true,PendulumStiffness=1.5,PendulumGravity=(X=0,Y=0,Z=0),PendulumBlend=0.95,PendulumDrag=0.98,PendulumMinimum=0,PendulumMaximum=0.7,PendulumEase=Linear,UnwindAxis=(X=0,Y=1,Z=0),UnwindMinimum=0.2,UnwindMaximum=0.05)
    Reach=(bEnabled=true,ReachTarget=(X=1.5,Y=1.5,Z=1.5),ReachAxis=(X=1,Y=0,Z=0),ReachMinimum=0,ReachMaximum=0,ReachEase=Linear)
    Speed=(X=3,Y=3,Z=3)
    Wave=(bEnabled=True,WaveFrequency=(X=0,Y=0,Z=0),WaveAmplitude=(X=0,Y=2,Z=0),WaveOffset=(X=0,Y=0,Z=0),WaveNoise=(X=0,Y=0,Z=0),WaveMinimum=5,WaveMaximum=5,WaveEase=SineEaseIn)
    WaveCurve="(EditorCurveData=(Keys=((InterpMode=RCIM_Cubic,Time=0.0121,Value=0.0862),(InterpMode=RCIM_Cubic,Time=0.278,Value=0.1784,ArriveTangent=0.0702,LeaveTangent=0.0702),(InterpMode=RCIM_Cubic,Time=0.5817,Value=0.1262,ArriveTangent=-0.0352,LeaveTangent=-0.0352),(I ...(truncated)"
    bDrawDebug=True
   )
  ]}
]
hierarchy [
 root {loc=<0 0 0>, rot=<0 0 0 1>}
 Melee_TwoHanded_Handle {parent=root, loc=<0 0 0>, rot=<0 0 0.7071 -0.7071>}
 dyn_handel_MSTR {parent=root, loc=<0 0 31.5526>, rot=<0.7071 0 0 -0.7071>}
 dyn_handel {parent=dyn_handel_MSTR, loc=<0 -0.7316 0>, rot=<0 0 0 1>}
 dyn_teeth_01 {parent=dyn_handel, loc=<-2.4551 -54.7032 3.3531>, rot=<-0.0789 0.0155 0.9791 -0.1866>}
 dyn_teeth_02 {parent=dyn_handel, loc=<-0.717 -43.4287 3.1701>, rot=<0.7065 -0.0068 0.7074 0.0187>}
 dyn_teeth_03 {parent=dyn_handel, loc=<0.2882 -35.1398 4.2856>, rot=<0.5033 0.5033 0.4967 -0.4967>}
 dyn_teeth_04 {parent=dyn_handel, loc=<4.4247 -47.432 -2.2456>, rot=<0.7066 0.2068 -0.6306 0.2455>}
 dyn_teeth_05 {parent=dyn_handel, loc=<-5.6712 -55.0227 -1.66>, rot=<0.7066 0.2068 -0.6306 0.2455>}
 dyn_teeth_06 {parent=dyn_handel, loc=<3.8726 -50.6584 4.3768>, rot=<-0.0769 0.0232 0.9558 -0.2829>}
 dyn_teeth_07 {parent=dyn_handel, loc=<-2.6126 -27.8492 4.1903>, rot=<-0.0774 -0.6963 -0.6927 0.1712>}
 dyn_teeth_08 {parent=dyn_handel, loc=<-2.8064 -31.3158 2.3894>, rot=<-0.1045 0.6439 0.6853 0.3237>}
 dyn_teeth_09 {parent=dyn_handel, loc=<-2.0505 -36.6672 -2.6485>, rot=<0.6081 0.3839 -0.5864 0.3728>}
 dyn_teeth_10 {parent=dyn_handel, loc=<1.1277 -29.9516 -3.9495>, rot=<0.6081 0.3839 -0.5864 0.3728>}
 dyn_teeth_11 {parent=dyn_handel, loc=<4.1191 -23.6199 -3.7656>, rot=<0.6514 0.4114 -0.5379 0.3422>}
 dyn_teeth_12 {parent=dyn_handel, loc=<5.9142 -16.9767 -2.5503>, rot=<0.6895 0.3439 -0.5008 0.3945>}
 dyn_teeth_13 {parent=dyn_handel, loc=<0.203 -13.7984 -4.16>, rot=<-0.5598 -0.4238 0.6244 -0.3421>}
 dyn_teeth_14 {parent=dyn_handel, loc=<-5.6921 -15.9746 -1.5155>, rot=<-0.2627 -0.6991 0.6649 0.0107>}
 dyn_teeth_15 {parent=dyn_handel, loc=<-2.997 -23.9579 -2.3771>, rot=<-0.1878 -0.6938 0.6755 -0.1647>}
 dyn_teeth_16 {parent=dyn_handel, loc=<1.0967 -23.1954 4.9198>, rot=<-0.5885 0.4129 -0.374 -0.586>}
 dyn_teeth_17 {parent=dyn_handel, loc=<-1.8561 -16.9202 4.5617>, rot=<0.3982 -0.5986 0.5614 0.4099>}
 dyn_belt_01 {parent=dyn_handel, loc=<2.9179 -0.3628 -4.1878>, rot=<0.0037 -0.4637 0.015 -0.8859>}
 dyn_teeth_handel_01_00 {parent=dyn_handel, loc=<-9.2521 9.7173 0.1099>, rot=<0.5086 0.495 -0.4965 -0.4998>}
 dyn_teeth_handel_01 {parent=dyn_teeth_handel_01_00, loc=<5.1537 -0.0488 -0.0833>, rot=<-0.0037 0.0085 0.0052 -0.9999>}
 dyn_teeth_handel_01_02 {parent=dyn_teeth_handel_01, loc=<5.9982 0 0>, rot=<0 0 0 1>}
 dyn_teeth_handel_02_00 {parent=dyn_handel, loc=<8.2348 9.4208 0.1151>, rot=<-0.5019 -0.4985 -0.493 -0.5066>}
 dyn_teeth_handel_02 {parent=dyn_teeth_handel_02_00, loc=<5.472 -0.0001 -0.3226>, rot=<0.0004 -0.0085 -0.0051 -1>}
 dyn_teeth_handel_02_02 {parent=dyn_teeth_handel_02, loc=<5.6268 0 0>, rot=<0 0 0 1>}
 dyn_teeth_handel_03_00 {parent=dyn_handel, loc=<3.776 9.3606 -7.8089>, rot=<0.6846 0.6729 0.1886 0.2071>}
 dyn_teeth_handel_03 {parent=dyn_teeth_handel_03_00, loc=<6.0724 -0.1292 -0.5329>, rot=<-0.0023 -0.0149 0.0043 -0.9999>}
 dyn_teeth_handel_03_02 {parent=dyn_teeth_handel_03, loc=<5.6268 0 0>, rot=<0 0 0 1>}
 dyn_teeth_handel_04_00 {parent=dyn_handel, loc=<-5.2204 9.5675 -7.8246>, rot=<0.6789 0.6652 -0.2213 -0.2182>}
 dyn_teeth_handel_04 {parent=dyn_teeth_handel_04_00, loc=<6.27 -0.2483 -0.2294>, rot=<-0.0086 0.001 0.0099 -0.9999>}
 dyn_teeth_handel_04_02 {parent=dyn_teeth_handel_04, loc=<5.6268 0 0>, rot=<0 0 0 1>}
 dyn_teeth_handel_05_00 {parent=dyn_handel, loc=<-5.1942 9.7267 7.7315>, rot=<0.1922 0.182 -0.677 -0.6867>}
 dyn_teeth_handel_05 {parent=dyn_teeth_handel_05_00, loc=<5.146 -0.1025 -0.0966>, rot=<-0.0148 0.0087 -0.0048 -0.9998>}
 dyn_teeth_handel_05_02 {parent=dyn_teeth_handel_05, loc=<5.9982 0 0>, rot=<0 0 0 1>}
 dyn_teeth_handel_06_00 {parent=dyn_handel, loc=<4.0813 9.5709 7.8755>, rot=<-0.182 -0.1859 -0.676 -0.6895>}
 dyn_teeth_handel_06 {parent=dyn_teeth_handel_06_00, loc=<4.8323 0.1035 0.0403>, rot=<-0.009 0.0001 -0.0099 -0.9999>}
 dyn_teeth_handel_06_02 {parent=dyn_teeth_handel_06, loc=<5.9982 0 0>, rot=<0 0 0 1>}
 dyn_tongue_01 {parent=dyn_handel_MSTR, loc=<-4.1949 -50.1117 -0.3654>, rot=<0.1053 -0.6992 0.6832 -0.1826>}
 dyn_tongue_02 {parent=dyn_tongue_01, loc=<5.1534 0 0>, rot=<0 -0.1703 0 -0.9854>}
 dyn_tongue_03 {parent=dyn_tongue_02, loc=<3.4675 0 0.2128>, rot=<0 -0.3025 0 -0.9531>}
 dyn_tongue_04 {parent=dyn_tongue_03, loc=<4.7117 -0.2641 -0.067>, rot=<0 0.1634 0 -0.9866>}
]
