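# Physics asset for the MelodyUrchin pickaxe, providing capsule bodies and joints for its dangling handle, belt, teeth and tongue bones.
# dyn_handel joins dyn_handel_MSTR with free swing and twist, while the unconstrained Melee_TwoHanded_Handle has its own body.
# A four-link tongue chain hangs off the master; the belt and six three-link teeth chains hang off dyn_handel.
# These joints barely twist, and teeth swing widens toward the tips.
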PhysicsAsset /BRCosmetics/Weapons/FORT_Melee/Pickaxe_MelodyUrchin/Meshes/MelodyUrchin_Axe_Physics
bodies {
 Melee_TwoHanded_Handle [{shapeType=Capsule, length=2, radius=2}]
 dyn_belt_01 [{shapeType=Capsule, length=1, radius=1, shapeName=dyn_belt_01_capsule, center=<1.5144 0 0>, rotation=<rot 90 0 0>}]
 dyn_handel [{shapeType=Capsule, length=2, radius=2}]
 dyn_handel_MSTR [{shapeType=Capsule, length=0.505, radius=0.505, shapeName=dyn_handel_MSTR_capsule, rotation=<rot 0 -90 0>}]
 dyn_teeth_handel_01 [{shapeType=Capsule, length=2.5, radius=1.5, shapeName=tongue_02_capsule, center=<3.1318 0 0>, rotation=<rot 90 -5.7682 -5.7676>}]
 dyn_teeth_handel_01_00 [{shapeType=Capsule, length=2.5, radius=1.5, shapeName=tongue_02_capsule, center=<3.1318 0 0>, rotation=<rot 90 -5.7682 -5.7676>}]
 dyn_teeth_handel_01_02 [{shapeType=Capsule, length=5, radius=2.5, shapeName=tongue_03_capsule, center=<2.5116 0 2.6896>}]
 dyn_teeth_handel_02 [{shapeType=Capsule, length=2.5, radius=1.5, shapeName=tongue_02_capsule, center=<3.1318 0 0>, rotation=<rot 90 0 -0.0006>}]
 dyn_teeth_handel_02_00 [{shapeType=Capsule, length=2.5, radius=1.5, shapeName=tongue_02_capsule, center=<3.1318 0 0>, rotation=<rot 90 -86.7424 -86.7419>}]
 dyn_teeth_handel_02_02 [{shapeType=Capsule, length=5, radius=2.5, shapeName=tongue_03_capsule, center=<2.5116 0 2.6896>}]
 dyn_teeth_handel_03 [{shapeType=Capsule, length=2.5, radius=1.5, shapeName=tongue_02_capsule, center=<3.1318 0 0>, rotation=<rot 90 0 -0.0003>}]
 dyn_teeth_handel_03_00 [{shapeType=Capsule, length=2.5, radius=1.5, shapeName=tongue_02_capsule, center=<3.1318 0 0>, rotation=<rot 90 -1.5791 -1.5785>}]
 dyn_teeth_handel_03_02 [{shapeType=Capsule, length=5, radius=2.5, shapeName=tongue_03_capsule, center=<2.5116 0 2.6896>}]
 dyn_teeth_handel_04 [{shapeType=Capsule, length=2.5, radius=1.5, shapeName=tongue_02_capsule, center=<3.1318 0 0>, rotation=<rot 90 0 0.0004>}]
 dyn_teeth_handel_04_00 [{shapeType=Capsule, length=2.5, radius=1.5, shapeName=tongue_02_capsule, center=<3.1318 0 0>, rotation=<rot 90 -5.7682 -5.7676>}]
 dyn_teeth_handel_04_02 [{shapeType=Capsule, length=5, radius=2.5, shapeName=tongue_03_capsule, center=<2.5116 0 2.6896>}]
 dyn_teeth_handel_05 [{shapeType=Capsule, length=2.5, radius=1.5, shapeName=tongue_02_capsule, center=<3.1318 0 0>, rotation=<rot 90 180 -179.9996>}]
 dyn_teeth_handel_05_00 [{shapeType=Capsule, length=2.5, radius=1.5, shapeName=tongue_02_capsule, center=<3.1318 0 0>, rotation=<rot 90 -5.7682 -5.7676>}]
 dyn_teeth_handel_05_02 [{shapeType=Capsule, length=5, radius=2.5, shapeName=tongue_03_capsule, center=<2.5116 0 2.6896>}]
 dyn_teeth_handel_06 [{shapeType=Capsule, length=2.5, radius=1.5, shapeName=tongue_02_capsule, center=<3.1318 0 0>, rotation=<rot 90 -180 179.9996>}]
 dyn_teeth_handel_06_00 [{shapeType=Capsule, length=2.5, radius=1.5, shapeName=tongue_02_capsule, center=<3.1318 0 0>, rotation=<rot 90 -178.8762 -178.8756>}]
 dyn_teeth_handel_06_02 [{shapeType=Capsule, length=5, radius=2.5, shapeName=tongue_03_capsule, center=<2.5116 0 2.6896>}]
 dyn_tongue_01 [{shapeType=Capsule, length=2.0501, radius=1.7083, shapeName=tongue_01_capsule, center=<2.584 0 0>, rotation=<rot 90 0 0>}]
 dyn_tongue_02 [{shapeType=Capsule, length=2.0501, radius=1.2083, shapeName=tongue_01_capsule, center=<2.3419 0 0>, rotation=<rot 90 0 0>}]
 dyn_tongue_03 [{shapeType=Capsule, length=1.0594, radius=1.2083, shapeName=tongue_03_capsule, center=<1.843 0 0>, rotation=<rot 90 0 0>}]
 dyn_tongue_04 [{shapeType=Capsule, length=1.0594, radius=1.2083, shapeName=tongue_03_capsule, center=<1.843 0 0>, rotation=<rot 90 0 0>}]
}
constraints [
 {bone1Name=dyn_teeth_handel_01, bone2Name=dyn_teeth_handel_01_00, swing1LimitDegrees=20, swing2LimitDegrees=20, twistLimitDegrees=1}
 {bone1Name=dyn_teeth_handel_01_02, bone2Name=dyn_teeth_handel_01, swing1LimitDegrees=20, swing2LimitDegrees=80, twistLimitDegrees=1}
 {bone1Name=dyn_teeth_handel_02, bone2Name=dyn_teeth_handel_02_00, swing1LimitDegrees=20, swing2LimitDegrees=50, twistLimitDegrees=1}
 {bone1Name=dyn_teeth_handel_02_02, bone2Name=dyn_teeth_handel_02, swing1LimitDegrees=20, swing2LimitDegrees=80, twistLimitDegrees=1}
 {bone1Name=dyn_teeth_handel_03, bone2Name=dyn_teeth_handel_03_00, swing1LimitDegrees=20, swing2LimitDegrees=40, twistLimitDegrees=1}
 {bone1Name=dyn_teeth_handel_03_02, bone2Name=dyn_teeth_handel_03, swing1LimitDegrees=20, swing2LimitDegrees=80, twistLimitDegrees=1}
 {bone1Name=dyn_teeth_handel_04, bone2Name=dyn_teeth_handel_04_00, swing1LimitDegrees=20, swing2LimitDegrees=40, twistLimitDegrees=1}
 {bone1Name=dyn_teeth_handel_04_02, bone2Name=dyn_teeth_handel_04, swing1LimitDegrees=20, swing2LimitDegrees=80, twistLimitDegrees=1}
 {bone1Name=dyn_teeth_handel_05, bone2Name=dyn_teeth_handel_05_00, swing1LimitDegrees=20, swing2LimitDegrees=40, twistLimitDegrees=1}
 {bone1Name=dyn_teeth_handel_05_02, bone2Name=dyn_teeth_handel_05, swing1LimitDegrees=20, swing2LimitDegrees=80, twistLimitDegrees=1}
 {bone1Name=dyn_teeth_handel_06, bone2Name=dyn_teeth_handel_06_00, swing1LimitDegrees=20, swing2LimitDegrees=40, twistLimitDegrees=1}
 {bone1Name=dyn_teeth_handel_06_02, bone2Name=dyn_teeth_handel_06, swing1LimitDegrees=20, swing2LimitDegrees=80, twistLimitDegrees=1}
 {bone1Name=dyn_tongue_02, bone2Name=dyn_tongue_01, swing1LimitDegrees=11, swing2LimitDegrees=11, twistLimitDegrees=1}
 {bone1Name=dyn_tongue_03, bone2Name=dyn_tongue_02, swing1LimitDegrees=12, swing2LimitDegrees=12, twistLimitDegrees=1}
 {bone1Name=dyn_tongue_04, bone2Name=dyn_tongue_03, swing1LimitDegrees=8, swing2LimitDegrees=8, twistLimitDegrees=1}
 {
  bone1Name=dyn_handel
  bone2Name=dyn_handel_MSTR
  swing1LimitDegrees=45
  swing1Motion=Free
  swing2LimitDegrees=45
  swing2Motion=Free
  twistLimitDegrees=45
  twistMotion=Free
 }
 {bone1Name=dyn_tongue_01, bone2Name=dyn_handel_MSTR, swing1LimitDegrees=15, swing2LimitDegrees=15, twistLimitDegrees=1}
 {bone1Name=dyn_belt_01, bone2Name=dyn_handel, swing1LimitDegrees=2, swing2LimitDegrees=2, twistLimitDegrees=1}
 {bone1Name=dyn_teeth_handel_01_00, bone2Name=dyn_handel, swing1LimitDegrees=10, swing2LimitDegrees=25, twistLimitDegrees=1}
 {bone1Name=dyn_teeth_handel_02_00, bone2Name=dyn_handel, swing1LimitDegrees=10, swing2LimitDegrees=25, twistLimitDegrees=1}
 {bone1Name=dyn_teeth_handel_03_00, bone2Name=dyn_handel, swing1LimitDegrees=10, swing2LimitDegrees=25, twistLimitDegrees=1}
 {bone1Name=dyn_teeth_handel_04_00, bone2Name=dyn_handel, swing1LimitDegrees=10, swing2LimitDegrees=25, twistLimitDegrees=1}
 {bone1Name=dyn_teeth_handel_05_00, bone2Name=dyn_handel, swing1LimitDegrees=10, swing2LimitDegrees=25, twistLimitDegrees=1}
 {bone1Name=dyn_teeth_handel_06_00, bone2Name=dyn_handel, swing1LimitDegrees=10, swing2LimitDegrees=25, twistLimitDegrees=1}
]
